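# Post-process style AnimBlueprint for the F_MED_SoloSnooze female cosmetic body, copying the attached parent's pose and adding secondary motion.
# A first Control Rig takes that pose and the parachute-open flag, then feeds sprung AnimDynamics on zipper and sleeve bones.
# Two zipper correction bones follow both thighs' rotation through constraints, before a correction Control Rig outputs the pose.
# The bottom-pants AnimDynamics chain is disconnected from the output; the event graph is empty.
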AnimBlueprint /BRCosmetics/Characters/Player/Female/Medium/Bodies/F_MED_SoloSnooze/Meshes/F_MED_SoloSnooze_AnimBP
target_skeleton=&/F_MED_SoloSnooze_Skeleton
graphs [
 AnimGraph {nodes [
   n1 = Anim Dynamics - Bone: dyn_bottompants_01_l(Alpha=1, ExternalForce="0,0,0") {
    angularDampingOverride=0.9
    angularSpringConstant=199
    bAngularSpring=true
    bOverrideAngularDamping=true
    bUseGravityOverride=true
    gravityScale=0.35
    simulationSpace=RootRelative
    boundBone {boneName=dyn_bottompants_01_l}
    componentAppliedLinearAccClamp=<3000 3000 3000>
    componentLinearAccScale=<1 1 1>
    componentLinearVelScale=<1 1 1>
    gravityOverride {z=-980}
    physicsBodyDefinitions [{
      boundBone {boneName=dyn_bottompants_01_l}
      boxExtents=<12 12 12>
      constraintSetup {angularLimitsMax=<20 20 20>, angularLimitsMin=<-20 -20 -20>, angularTarget=<1 0 0>, linearAxesMax=<1 1 1>, linearAxesMin=<-1 -1 -1>}
      localJointOffset=<-3 0 0>
     }]
   }
   n2 = Anim Dynamics - Bone: dyn_bottompants_02_l(ComponentPose=@n1, Alpha=1, ExternalForce="0,0,0") {
    angularDampingOverride=0.9
    angularSpringConstant=199
    bAngularSpring=true
    bOverrideAngularDamping=true
    bUseGravityOverride=true
    gravityScale=0.35
    simulationSpace=RootRelative
    boundBone {boneName=dyn_bottompants_02_l}
    componentAppliedLinearAccClamp=<3000 3000 3000>
    componentLinearAccScale=<1 1 1>
    componentLinearVelScale=<1 1 1>
    gravityOverride {z=-980}
    physicsBodyDefinitions [{
      boundBone {boneName=dyn_bottompants_02_l}
      boxExtents=<12 12 12>
      constraintSetup {angularLimitsMax=<20 20 20>, angularLimitsMin=<-20 -20 -20>, angularTarget=<1 0 0>, linearAxesMin=<-1 -1 -1>}
      localJointOffset=<-3 0 0>
     }]
   }
   n3 = Anim Dynamics - Bone: dyn_bottompants_01_r(ComponentPose=@n2, Alpha=1, ExternalForce="0,0,0") {
    angularDampingOverride=0.9
    angularSpringConstant=199
    bAngularSpring=true
    bOverrideAngularDamping=true
    bUseGravityOverride=true
    gravityScale=0.35
    simulationSpace=RootRelative
    boundBone {boneName=dyn_bottompants_01_r}
    componentAppliedLinearAccClamp=<3000 3000 3000>
    componentLinearAccScale=<1 1 1>
    componentLinearVelScale=<1 1 1>
    gravityOverride {z=-980}
    physicsBodyDefinitions [{
      boundBone {boneName=dyn_bottompants_01_r}
      boxExtents=<12 12 12>
      constraintSetup {angularLimitsMax=<20 20 20>, angularLimitsMin=<-20 -20 -20>, angularTarget=<1 0 0>, linearAxesMax=<1 1 1>, linearAxesMin=<-1 -1 -1>}
      localJointOffset=<-3 0 0>
     }]
   }
   Anim Dynamics - Bone: dyn_bottompants_02_r(ComponentPose=@n3, Alpha=1, ExternalForce="0,0,0") {
    angularDampingOverride=0.9
    angularSpringConstant=199
    bAngularSpring=true
    bOverrideAngularDamping=true
    bUseGravityOverride=true
    gravityScale=0.35
    simulationSpace=RootRelative
    boundBone {boneName=dyn_bottompants_02_r}
    componentAppliedLinearAccClamp=<3000 3000 3000>
    componentLinearAccScale=<1 1 1>
    componentLinearVelScale=<1 1 1>
    gravityOverride {z=-980}
    physicsBodyDefinitions [{
      boundBone {boneName=dyn_bottompants_02_r}
      boxExtents=<12 12 12>
      constraintSetup {angularLimitsMax=<20 20 20>, angularLimitsMin=<-20 -20 -20>, angularTarget=<1 0 0>, linearAxesMax=<1 1 1>, linearAxesMin=<-1 -1 -1>}
      localJointOffset=<-3 0 0>
     }]
   }
   n4 = Anim Dynamics - Bone: dyn_zipper(ComponentPose=Local To Component(LocalPose=Use cached pose 'Control Rig'()), Alpha=1) {
    angularDampingOverride=0.999
    angularSpringConstant=320
    bAngularSpring=true
    bOverrideAngularDamping=true
    bUseGravityOverride=true
    gravityScale=0.35
    simulationSpace=RootRelative
    boundBone {boneName=dyn_zipper}
    componentAppliedLinearAccClamp=<3000 3000 3000>
    componentLinearAccScale=<0.9 0.9 0.9>
    componentLinearVelScale=<0.9 0.9 0.9>
    gravityOverride {z=-980}
    physicsBodyDefinitions [{
      boundBone {boneName=dyn_zipper}
      boxExtents=<13 13 9>
      constraintSetup {angularLimitsMax=<0 55 15>, angularLimitsMin=<0 -2 -15>, angularTarget=<1 0 0>}
      localJointOffset=<-1 0 0>
     }]
   }
   n5 = Anim Dynamics - Bone: dyn_sleeve_01_l(ComponentPose=@n4, Alpha=1, ExternalForce="0,0,0") {
    angularDampingOverride=0.9
    angularSpringConstant=400
    bAngularSpring=true
    bOverrideAngularDamping=true
    bUseGravityOverride=true
    gravityScale=0.35
    simulationSpace=RootRelative
    boundBone {boneName=dyn_sleeve_01_l}
    componentAppliedLinearAccClamp=<3000 3000 3000>
    componentLinearAccScale=<1 1 1>
    componentLinearVelScale=<1 1 1>
    gravityOverride {z=-980}
    physicsBodyDefinitions [{
      boundBone {boneName=dyn_sleeve_01_l}
      boxExtents=<10 10 10>
      constraintSetup {angularLimitsMax=<20 20 20>, angularLimitsMin=<-20 -20 -20>, angularTarget=<1 0 0>}
      localJointOffset=<0 -1 0>
     }]
   }
   n6 = Anim Dynamics - Bone: dyn_sleeve_02_l(ComponentPose=@n5, Alpha=1, ExternalForce="0,0,0") {
    angularDampingOverride=0.9
    angularSpringConstant=400
    bAngularSpring=true
    bOverrideAngularDamping=true
    bUseGravityOverride=true
    gravityScale=0.35
    simulationSpace=RootRelative
    boundBone {boneName=dyn_sleeve_02_l}
    componentAppliedLinearAccClamp=<3000 3000 3000>
    componentLinearAccScale=<1 1 1>
    componentLinearVelScale=<1 1 1>
    gravityOverride {z=-980}
    physicsBodyDefinitions [{
      boundBone {boneName=dyn_sleeve_02_l}
      boxExtents=<10 10 10>
      constraintSetup {angularLimitsMax=<20 20 20>, angularLimitsMin=<-20 -20 -20>, angularTarget=<1 0 0>}
      localJointOffset=<0 -1 0>
     }]
   }
   n7 = Anim Dynamics - Bone: dyn_sleeve_01_r(ComponentPose=@n6, Alpha=1, ExternalForce="0,0,0") {
    angularDampingOverride=0.9
    angularSpringConstant=400
    bAngularSpring=true
    bOverrideAngularDamping=true
    bUseGravityOverride=true
    gravityScale=0.35
    simulationSpace=RootRelative
    boundBone {boneName=dyn_sleeve_01_r}
    componentAppliedLinearAccClamp=<3000 3000 3000>
    componentLinearAccScale=<1 1 1>
    componentLinearVelScale=<1 1 1>
    gravityOverride {z=-980}
    physicsBodyDefinitions [{
      boundBone {boneName=dyn_sleeve_01_r}
      boxExtents=<10 10 10>
      constraintSetup {angularLimitsMax=<20 20 20>, angularLimitsMin=<-20 -20 -20>, angularTarget=<1 0 0>}
      localJointOffset=<1 0 0>
     }]
   }
   n8 = Anim Dynamics - Bone: dyn_sleeve_02_r(ComponentPose=@n7, Alpha=1, ExternalForce="0,0,0") {
    angularDampingOverride=0.9
    angularSpringConstant=400
    bAngularSpring=true
    bOverrideAngularDamping=true
    bUseGravityOverride=true
    gravityScale=0.35
    simulationSpace=RootRelative
    boundBone {boneName=dyn_sleeve_02_r}
    componentAppliedLinearAccClamp=<3000 3000 3000>
    componentLinearAccScale=<1 1 1>
    componentLinearVelScale=<1 1 1>
    gravityOverride {z=-980}
    physicsBodyDefinitions [{
      boundBone {boneName=dyn_sleeve_02_r}
      boxExtents=<12 12 12>
      constraintSetup {angularLimitsMax=<20 20 20>, angularLimitsMin=<-20 -20 -20>, angularTarget=<1 0 0>}
      localJointOffset=<1 0 0>
     }]
   }
   n9 = "Constraint - MSTR_zipper_correction_01 (RR)"!(ComponentPose=@n8, Alpha=1, ConstraintWeights_0=0.5, ConstraintWeights_1:float=0.5) -> (Pose) {
    boneToModify {boneName=MSTR_zipper_correction_01}
    constraintSetup [{transformType=Rotation, targetBone {boneName=thigh_l}}, {transformType=Rotation, targetBone {boneName=thigh_r}}]
    constraintWeights [1,1]
   }
   n10 = "Constraint - MSTR_zipper_correction_02 (RR)"!(ComponentPose=@n9, Alpha=1, ConstraintWeights_0=0.5, ConstraintWeights_1:float=0.5) -> (Pose) {
    boneToModify {boneName=MSTR_zipper_correction_02}
    constraintSetup [{transformType=Rotation, targetBone {boneName=thigh_l}}, {transformType=Rotation, targetBone {boneName=thigh_r}}]
    constraintWeights [1,1]
   }
   n11 = Control Rig(Source=Component To Local(ComponentPose=@n10)) {controlRigAssetReference {blueprintRigClass=&/F_MED_SoloSnooze_Correction_CtrlRig.F_MED_SoloSnooze_Correction_CtrlRig_C}}
   Output Pose(Result=@n11) {layerGroup=None}
   n12 = Control Rig!(Source=Copy Pose From Mesh() {bUseAttachedParent=true}, Alpha, "Parachute opened ":bool=Get bIsParachuteOpened()) -> (Pose) {controlRigAssetReference {blueprintRigClass=&/F_MED_SoloSnooze_CtrlRig.F_MED_SoloSnooze_CtrlRig_C}}
   Save cached pose 'Control Rig'(Pose=@n12) {cacheName=Control Rig}
  ]}
 EventGraph {nodes [Event BlueprintUpdateAnimation()]}
]
variables []
