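# Physics asset for the F_MED_SoloSnooze female cosmetic's face accessory, setting up physics for two ears and a pair of laces.
# Each ear is constrained to a tiny MSTR anchor body with modest swing and twist limits.
# Each lace is a three-link chain from its own anchor, tighter at the root, with zero twist limit.
# Unconstrained capsules cover the head, upper spine and right arm only, with none on the left arm.
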PhysicsAsset /BRCosmetics/Characters/Player/Female/Medium/Bodies/F_MED_SoloSnooze/Meshes/Parts/F_MED_SoloSnooze_FaceAcc_Physics
bodies {
 MSTR_ear_l [{shapeType=Capsule, length=0.1, radius=0.1, shapeName=MSTR_ear_l_capsule, rotation=<rot 0 -90 0>}]
 MSTR_ear_r [{shapeType=Capsule, length=0.1, radius=0.1, shapeName=MSTR_ear_r_capsule, rotation=<rot 0 -90 0>}]
 MSTR_laces_l [{shapeType=Capsule, length=0.1, radius=0.1, shapeName=MSTR_laces_l_capsule, rotation=<rot 0 -90 0>}]
 MSTR_laces_r [{shapeType=Capsule, length=0.505, radius=0.505, shapeName=MSTR_laces_r_capsule, rotation=<rot 0 -90 0>}]
 dyn_ear_l [{shapeType=Capsule, length=5.513, radius=4.8569, shapeName=dyn_ear_l_capsule, rotation=<rot 90 0 0>}]
 dyn_ear_r [{shapeType=Capsule, length=5.513, radius=4.8569, shapeName=dyn_ear_r_capsule}]
 dyn_laces_01_l [{
   shapeType=Capsule
   length=4.5676
   radius=0.7333
   shapeName=dyn_laces_01_l_capsule
   center=<3.1709 -0.022 -0.0872>
   rotation=<rot 0 90.6237 -87.4621>
  }]
 dyn_laces_01_r [{
   shapeType=Capsule
   length=4.6761
   radius=0.7435
   shapeName=dyn_laces_01_r_capsule
   center=<3.1709 0.022 0.0872>
   rotation=<rot 0 90.6237 92.5379>
  }]
 dyn_laces_02_l [{shapeType=Capsule, length=4.333, radius=0.6134, shapeName=dyn_laces_02_l_capsule, center=<2.2421 0 0>, rotation=<rot 0 -89.3503 -90.0261>}]
 dyn_laces_02_r [{shapeType=Capsule, length=4.333, radius=0.6134, shapeName=dyn_laces_02_r_capsule, center=<2.2421 0 0>, rotation=<rot 0 -89.3503 86.8977>}]
 dyn_laces_03_l [{
   shapeType=Capsule
   length=3.9183
   radius=1.3159
   shapeName=dyn_laces_03_l_capsule
   center=<4.2037 0.1075 -0.2336>
   rotation=<rot 0 -86.7497 -77.6402>
  }]
 dyn_laces_03_r [{
   shapeType=Capsule
   length=4.1149
   radius=1.1194
   shapeName=dyn_laces_03_r_capsule
   center=<3.9782 0.2969 -0.18>
   rotation=<rot 0 -86.7497 102.3598>
  }]
 head [{shapeType=Capsule, length=3.1247, radius=8.9828, shapeName=head_capsule, rotation=<rot 90 0 0>}]
 lowerarm_r [{shapeType=Capsule, length=18.2326, radius=5.4036, shapeName=lowerarm_r_capsule, center=<-12.8828 0 0>, rotation=<rot 90 0 0>}]
 spine_04 [{shapeType=Capsule, length=6.6443, radius=13.8865, shapeName=spine_04_capsule, center=<-2.1189 0.4282 0>}]
 spine_05 [{shapeType=Capsule, length=12.5246, radius=9.8802, shapeName=spine_05_capsule, center=<-4.8017 1.6848 0>}]
 upperarm_r [{shapeType=Capsule, length=19.6534, radius=5.3164, shapeName=upperarm_r_capsule, center=<-12.8744 0 0>, rotation=<rot 90 0 0>}]
}
constraints [
 {bone1Name=dyn_ear_l, bone2Name=MSTR_ear_l, swing1LimitDegrees=18, swing2LimitDegrees=18, twistLimitDegrees=18}
 {bone1Name=dyn_ear_r, bone2Name=MSTR_ear_r, swing1LimitDegrees=18, swing2LimitDegrees=18, twistLimitDegrees=18}
 {bone1Name=dyn_laces_01_l, bone2Name=MSTR_laces_l, swing1LimitDegrees=15, swing2LimitDegrees=10, twistLimitDegrees=0}
 {bone1Name=dyn_laces_02_l, bone2Name=dyn_laces_01_l, swing1LimitDegrees=45, swing2LimitDegrees=45, twistLimitDegrees=0}
 {bone1Name=dyn_laces_03_l, bone2Name=dyn_laces_02_l, swing1LimitDegrees=45, swing2LimitDegrees=45, twistLimitDegrees=0}
 {bone1Name=dyn_laces_01_r, bone2Name=MSTR_laces_r, swing1LimitDegrees=15, swing2LimitDegrees=10, twistLimitDegrees=0}
 {bone1Name=dyn_laces_02_r, bone2Name=dyn_laces_01_r, swing1LimitDegrees=45, swing2LimitDegrees=45, twistLimitDegrees=0}
 {bone1Name=dyn_laces_03_r, bone2Name=dyn_laces_02_r, swing1LimitDegrees=45, swing2LimitDegrees=45, twistLimitDegrees=0}
]
